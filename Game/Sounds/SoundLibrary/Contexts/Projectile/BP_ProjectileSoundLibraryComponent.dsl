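; Blueprint sound library component for projectiles that manages the end of their in-air sound.
; On EndPlay it calls the parent, then stops the InAir event so the loop does not stick when one projectile lands on another.
; Its OnSoundStopped function exact-matches the InAir gameplay tag, linearly fades out a valid audio component and returns true.
; For other tags it returns false.
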(event EventEndPlay
  (params (EndPlayReason :type EEndPlayReason%20Enum :default "Destroyed"))
  ; Fixes issue with stuck in air loop if a projectile lands on top of another
  (|CallParentFunction EndPlayReason)
  (SoundLibrary|StopEvent "(TagName=\"SoundLibrary.Projectile.InAir\")"))

(fn OnSoundStopped
  (params (InEventName :type Gameplay%20Tag%20Structure) (InComponent :type Audio%20Component%20Object%20Reference))
  (bind _incomponent (Variables|GetInComponent))
  (if (GameplayTags|MatchesTag InEventName "(TagName=\"SoundLibrary.Projectile.InAir\")" true)
    (Utilities|IsValid _incomponent
      (:Is%20Valid
        (Audio|Components|Audio|FadeOut _incomponent 0.8 0.0 "Linear" :!class /Script/Engine.AudioComponent)
        (return true)))
    (else
      (return false))))
